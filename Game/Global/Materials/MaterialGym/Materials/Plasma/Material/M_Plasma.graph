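# M_Plasma is an unlit material for a plasma effect, stored in the MaterialGym folder.
# Its graph holds a single call to the MF_Plasma material function, so the effect logic lives in that function rather than here.
# It flags pixel animation, declares two customized UV channels and uses legacy pixel depth offset.
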Material /Game/Global/Materials/MaterialGym/Materials/Plasma/Material/M_Plasma
graph {
 properties {bHasPixelAnimation=true, numCustomizedUVs=2, pixelDepthOffsetMode=PDOM_Legacy, shadingModel=MSM_Unlit, usedShadingModels=Unlit}
 nodes [MaterialFunctionCall() {materialFunction=/Game/Global/Materials/MaterialGym/Materials/Plasma/MaterialFunctions/MF_Plasma}]
}
