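# Material function that builds an animated plasma color effect of repeated wavy strands with glow and colored ripples.
# Bump-offset UVs are tiled by strand count and displaced by a time-driven sine, perturbed by panning noise texture, to form a distance field.
# Sphere masks shape strands and ripple bands, with two time-scrolled, squared-Frac ripple layers and a sine-pulsed glow.
# The output combines these color layers through Max and Add.
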MaterialFunction /Game/Global/Materials/MaterialGym/Materials/Plasma/MaterialFunctions/MF_PlasmaNoise1
graph {nodes [
  n1 = NamedRerouteUsage()
  n2 = NamedRerouteUsage()
  FunctionOutput(None=Add(A=Max(A=@n1, B=@n2), B=NamedRerouteUsage()))
  n3 = VectorParameter() {group=Plasma, parameterName=Plasma_RippleColor_01, sortPriority=33, defaultValue=<rgba 0 0.2303 1 1>}
  n4 = NamedRerouteUsage()
  n5 = Multiply(A=Time(), B=ScalarParameter() {defaultValue=1, group=Plasma, parameterName=Plasma_StrandRippleSpeed_01, sortPriority=31})
  n6 = Frac(None=Multiply(A=Add(A=@n4, B=@n5), B=ScalarParameter() {defaultValue=7, group=Plasma, parameterName=Plasma_StrandRippleCount_01}) {constB=7})
  n7 = Multiply(A=@n3.RGB, B=Multiply(A=NamedRerouteUsage(), B=Multiply(A=@n6, B=@n6)))
  n8 = VectorParameter() {group=Plasma, parameterName=Plasma_RippleColor_02, sortPriority=42, defaultValue=<rgba 0.4612 0 1 1>}
  n9 = NamedRerouteUsage()
  n10 = Multiply(A=Time(), B=ScalarParameter() {defaultValue=0.75, group=Plasma, parameterName=Plasma_StrandRippleSpeed_02, sortPriority=40})
  n11 = Frac(None=Multiply(A=Add(A=@n9, B=@n10), B=ScalarParameter() {defaultValue=3, group=Plasma, parameterName=Plasma_StrandRippleCount_02, sortPriority=41}) {constB=3})
  n12 = Multiply(A=@n8.RGB, B=Multiply(A=NamedRerouteUsage(), B=Multiply(A=@n11, B=@n11)))
  NamedRerouteDeclaration(None=Multiply(A=Add(A=@n7, B=@n12), B=ScalarParameter() {defaultValue=0.5, group=Plasma, parameterName=Plasma_RippleBrightness, sortPriority=29})) {name=Ripples, nodeColor=<rgba 0.4531 0 1 1>}
  NamedRerouteDeclaration(None=FunctionInput(Preview=TextureCoordinate()) {bUsePreviewValueAsDefault=true, inputName=UV, inputType=FunctionInput_Vector2}) {name=UVInput, nodeColor=<rgba 0.2422 0 1 1>}
  n13 = BumpOffset(
   Coordinate=NamedRerouteUsage()
   Height=Constant()
   HeightRatioInput=ScalarParameter() {defaultValue=0.5, group=Plasma, parameterName=Plasma_BumpOffset, sortPriority=0}
  ) {heightRatio=0}
  NamedRerouteDeclaration(None=ComponentMask(None=@n13) {r=true}) {name=UVX, nodeColor=<rgba 0.0312 0 1 1>}
  NamedRerouteDeclaration(None=ComponentMask(None=@n13) {g=true}) {name=UVY, nodeColor=<rgba 1 0 0.375 1>}
  n14 = Multiply(
   A=NamedRerouteUsage()
   B=NamedRerouteDeclaration(None=ScalarParameter() {defaultValue=2, group=Plasma, parameterName=Plasma_StrandCount, sortPriority=1}) {name=StrandCount, nodeColor=<rgba 0 1 0.8125 1>}
  ) {constB=3}
  NamedRerouteDeclaration(None=Subtract(A=Frac(None=@n14), B=NamedRerouteUsage())) {name=DF, nodeColor=<rgba 0.3828 0 1 1>}
  n15 = ScalarParameter() {defaultValue=0.065, group=Plasma, parameterName=Plasma_RippleThickness, sortPriority=30}
  n16 = NamedRerouteUsage()
  NamedRerouteDeclaration(None=SphereMask(A=NamedRerouteUsage(), B=Constant() {r=0.5}, Radius=Multiply(A=@n15, B=@n16)) {hardnessPercent=0}) {name=RippleMask, nodeColor=<rgba 0.3594 1 0 1>}
  n17 = Multiply(A=Time(), B=ScalarParameter() {defaultValue=0.1, group=Plasma, parameterName=Plasma_Strand_Wave_Speed, sortPriority=60}) {constB=0.25}
  n18 = NamedRerouteUsage()
  n19 = ScalarParameter() {defaultValue=1, group=Plasma, parameterName=Plasma_Strand_Noise_Tiling, sortPriority=80}
  n20 = ScalarParameter() {group=Plasma, parameterName=Plasma_Strand_Noise_SpeedX, sortPriority=81}
  n21 = ScalarParameter() {defaultValue=0.2, group=Plasma, parameterName=Plasma_Strand_Noise_SpeedY, sortPriority=82}
  n22 = TextureSampleParameter2D(UVs=Panner(
    Coordinate=Multiply(A=@n18, B=@n19)
    Speed=MaterialFunctionCall(X=@n20, Y=@n21) {materialFunction=&MakeFloat2}.Result
   ) {speedY=0.4}) {
   group=Plasma
   parameterName=Plasma_NoiseTexture
   samplerType=SAMPLERTYPE_Masks
   sortPriority=70
   texture=/Game/Global/Textures/Default/Noise/T_WrapEdgeWear_Noise
  }
  n23 = Multiply(
   A=ScalarParameter() {defaultValue=0.2, group=Plasma, parameterName=Plasma_Strand_Noise_Strength, sortPriority=85}
   B=NamedRerouteDeclaration(None=ChannelMaskParameter(None=@n22.RGBA) {group=Plasma, parameterName=Plasma_NoiseTextureChannel, sortPriority=71}.R) {name=NoiseSample, nodeColor=<rgba 0.3594 1 0 1>}
  )
  n24 = Multiply(
   A=Add(A=NamedRerouteUsage(), B=Add(A=@n17, B=@n23))
   B=ScalarParameter() {defaultValue=2, group=Plasma, parameterName=Plasma_WaveFrequency, sortPriority=61}
  ) {constB=5}
  n25 = Divide(B=NamedRerouteUsage()) {constA=0.5}
  n26 = LinearInterpolate(
   A=Min(A=@n25, B=ScalarParameter() {defaultValue=0.05, group=Plasma, parameterName=Plasma_Amplitude_Min, sortPriority=65})
   B=Min(A=@n25, B=ScalarParameter() {defaultValue=0.1, group=Plasma, parameterName=Plasma_Amplitude_Max, sortPriority=66})
   Alpha=NamedRerouteUsage()
  )
  n27 = NamedRerouteUsage()
  NamedRerouteDeclaration(None=Multiply(A=Sine(None=@n24), B=Multiply(A=@n26, B=@n27)) {constB=0.1}) {name=Noise, nodeColor=<rgba 0 0.4375 1 1>}
  n28 = VectorParameter() {group=Plasma, parameterName=Plasma_StrandGlowColor, sortPriority=15, defaultValue=<rgba 1 0.5161 0 1>}
  n29 = Divide(A=NamedRerouteUsage()) {constB=48}
  n30 = ScalarParameter() {defaultValue=2, group=Plasma, parameterName=Plasma_StrandGlowBrightness, sortPriority=16}
  n31 = ConstantBiasScale(None=Sine(None=Time()))
  NamedRerouteDeclaration(None=Multiply(A=Multiply(A=@n28.RGB, B=@n29), B=Multiply(A=@n30, B=@n31))) {name=StrandGlow, nodeColor=<rgba 0.5937 1 0 1>}
  n32 = VectorParameter() {group=Plasma, parameterName=Plasma_StrandColor, sortPriority=11, defaultValue=<rgba 1 0.1992 0 1>}
  n33 = SphereMask(
   A=NamedRerouteUsage()
   B=Constant() {r=0.5}
   Radius=Multiply(A=ScalarParameter() {defaultValue=0.004, group=Plasma, parameterName=Plasma_StrandThickness, sortPriority=10}, B=NamedRerouteUsage())
  ) {hardnessPercent=0}
  n34 = Multiply(A=@n33) {constB=2}
  n35 = Multiply(A=ConstantBiasScale(None=Sine(None=@n34)), B=@n33)
  NamedRerouteDeclaration(None=Multiply(A=Multiply(A=@n32.RGB, B=@n35), B=ScalarParameter() {defaultValue=25, group=Plasma, parameterName=Plasma_StrandBrightness, sortPriority=12})) {name=Strand, nodeColor=<rgba 1 0.4453 0 1>}
  TextureCoordinate()
  TextureCoordinate()
 ]}
