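# M_PlasmaExample is an unlit example material for a plasma effect in the MaterialGym material library.
# Its graph contains a single call to the MF_PlasmaNoise1 material function, with no other nodes or visible connections.
# The material is flagged as having pixel animation and declares two customized UV channels.
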Material /Game/Global/Materials/MaterialGym/Materials/Plasma/Material/M_PlasmaExample
graph {
 properties {bHasPixelAnimation=true, numCustomizedUVs=2, pixelDepthOffsetMode=PDOM_Legacy, shadingModel=MSM_Unlit, usedShadingModels=Unlit}
 nodes [MaterialFunctionCall!(UV) {materialFunction=/Game/Global/Materials/MaterialGym/Materials/Plasma/MaterialFunctions/MF_PlasmaNoise1}]
}
